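# Anim Blueprint for the PepperBilly pickaxe that adds physics-driven secondary motion to its fifteen dyn_Curl bones.
# The cached input pose feeds three serial AnimDynamics chains with per-bone angular limits, selected by bool blends.
# In the front end, bIsOnPawn picks between two world-space chains differing in joint offset and damping.
# Otherwise, a root-relative chain with stiff angular springs and tighter acceleration clamps runs; the event graph has no logic.
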AnimBlueprint /BRCosmetics/Weapons/FORT_Melee/Pickaxe_PepperBilly/Meshes/PepperBilly_Axe_AnimBP
target_skeleton=&/PepperBilly_Axe_Skeleton
graphs [
 AnimGraph {nodes [
   n1 = Anim Dynamics - Bone: dyn_Curl_01(ComponentPose=Local To Component(LocalPose=Use cached pose 'Input Pose Cache'()), Alpha=1, ExternalForce="0,0,0") {
    angularDampingOverride=0.999
    angularSpringConstant=150
    bOverrideAngularDamping=true
    bUseGravityOverride=true
    bUsePlanarLimit=false
    gravityScale=0.35
    numSolverIterationsPreUpdate=1
    simulationSpace=World
    alphaBoolBlend {blendOption=Linear}
    boundBone {boneName=dyn_Curl_01}
    componentAppliedLinearAccClamp=<5000 5000 5000>
    componentLinearAccScale=<0.5 0.5 0.5>
    componentLinearVelScale=<0.5 0.5 0.5>
    gravityOverride {z=-980}
    physicsBodyDefinitions [{
      boundBone {boneName=dyn_Curl_01}
      boxExtents=<5 5 5>
      constraintSetup {angularLimitsMax=<20 30 20>, angularLimitsMin=<-10 -50 -20>, angularTarget=<1 0 0>}
      localJointOffset=<0.2 0 0.2>
     }]
   }
   n2 = Anim Dynamics - Bone: dyn_Curl_02(ComponentPose=@n1, Alpha=1, ExternalForce="0,0,0") {
    angularDampingOverride=0.999
    angularSpringConstant=150
    bOverrideAngularDamping=true
    bUseGravityOverride=true
    bUsePlanarLimit=false
    gravityScale=0.35
    numSolverIterationsPreUpdate=1
    simulationSpace=World
    alphaBoolBlend {blendOption=Linear}
    boundBone {boneName=dyn_Curl_02}
    componentAppliedLinearAccClamp=<5000 5000 5000>
    componentLinearAccScale=<0.5 0.5 0.5>
    componentLinearVelScale=<0.5 0.5 0.5>
    gravityOverride {z=-980}
    physicsBodyDefinitions [{
      boundBone {boneName=dyn_Curl_02}
      boxExtents=<10 10 10>
      constraintSetup {angularLimitsMax=<20 50 20>, angularLimitsMin=<-10 -50 -20>, angularTarget=<1 0 0>}
      localJointOffset=<0.2 0 0>
     }]
   }
   n3 = Anim Dynamics - Bone: dyn_Curl_03(ComponentPose=@n2, Alpha=1, ExternalForce="0,0,0") {
    angularDampingOverride=0.999
    angularSpringConstant=150
    bOverrideAngularDamping=true
    bUseGravityOverride=true
    bUsePlanarLimit=false
    gravityScale=0.35
    numSolverIterationsPreUpdate=1
    simulationSpace=World
    alphaBoolBlend {blendOption=Linear}
    boundBone {boneName=dyn_Curl_03}
    componentAppliedLinearAccClamp=<5000 5000 5000>
    componentLinearAccScale=<0.5 0.5 0.5>
    componentLinearVelScale=<0.5 0.5 0.5>
    gravityOverride {z=-980}
    physicsBodyDefinitions [{
      boundBone {boneName=dyn_Curl_03}
      boxExtents=<2 2 2>
      constraintSetup {angularLimitsMax=<20 0 20>, angularLimitsMin=<-10 -50 -20>, angularTarget=<1 0 0>}
      localJointOffset=<0.2 0 0>
     }]
   }
   n4 = Anim Dynamics - Bone: dyn_Curl_04(ComponentPose=@n3, Alpha=1, ExternalForce="0,0,0") {
    angularDampingOverride=0.999
    angularSpringConstant=150
    bOverrideAngularDamping=true
    bUseGravityOverride=true
    bUsePlanarLimit=false
    gravityScale=0.35
    numSolverIterationsPreUpdate=1
    simulationSpace=World
    alphaBoolBlend {blendOption=Linear}
    boundBone {boneName=dyn_Curl_04}
    componentAppliedLinearAccClamp=<5000 5000 5000>
    componentLinearAccScale=<0.5 0.5 0.5>
    componentLinearVelScale=<0.5 0.5 0.5>
    gravityOverride {z=-980}
    physicsBodyDefinitions [{
      boundBone {boneName=dyn_Curl_04}
      boxExtents=<5 5 5>
      constraintSetup {angularLimitsMax=<20 30 20>, angularLimitsMin=<-10 -40 -20>, angularTarget=<1 0 0>}
      localJointOffset=<0.2 0 0>
     }]
   }
   n5 = Anim Dynamics - Bone: dyn_Curl_05(ComponentPose=@n4, Alpha=1, ExternalForce="0,0,0") {
    angularDampingOverride=0.999
    angularSpringConstant=150
    bOverrideAngularDamping=true
    bUseGravityOverride=true
    bUsePlanarLimit=false
    gravityScale=0.35
    numSolverIterationsPreUpdate=1
    simulationSpace=World
    alphaBoolBlend {blendOption=Linear}
    boundBone {boneName=dyn_Curl_05}
    componentAppliedLinearAccClamp=<5000 5000 5000>
    componentLinearAccScale=<0.5 0.5 0.5>
    componentLinearVelScale=<0.5 0.5 0.5>
    gravityOverride {z=-980}
    physicsBodyDefinitions [{
      boundBone {boneName=dyn_Curl_05}
      boxExtents=<10 10 10>
      constraintSetup {angularLimitsMax=<20 50 20>, angularLimitsMin=<-10 -50 -20>, angularTarget=<1 0 0>}
      localJointOffset=<0.2 0 0>
     }]
   }
   n6 = Anim Dynamics - Bone: dyn_Curl_06(ComponentPose=@n5, Alpha=1, ExternalForce="0,0,0") {
    angularDampingOverride=0.999
    angularSpringConstant=150
    bOverrideAngularDamping=true
    bUseGravityOverride=true
    bUsePlanarLimit=false
    gravityScale=0.35
    numSolverIterationsPreUpdate=1
    simulationSpace=World
    alphaBoolBlend {blendOption=Linear}
    boundBone {boneName=dyn_Curl_06}
    componentAppliedLinearAccClamp=<5000 5000 5000>
    componentLinearAccScale=<0.5 0.5 0.5>
    componentLinearVelScale=<0.5 0.5 0.5>
    gravityOverride {z=-980}
    physicsBodyDefinitions [{
      boundBone {boneName=dyn_Curl_06}
      boxExtents=<2 2 2>
      constraintSetup {angularLimitsMax=<20 30 30>, angularLimitsMin=<-10 -30 -30>, angularTarget=<1 0 0>}
      localJointOffset=<0.2 0 0>
     }]
   }
   n7 = Anim Dynamics - Bone: dyn_Curl_07(ComponentPose=@n6, Alpha=1, ExternalForce="0,0,0") {
    angularDampingOverride=0.999
    angularSpringConstant=150
    bOverrideAngularDamping=true
    bUseGravityOverride=true
    bUsePlanarLimit=false
    gravityScale=0.35
    numSolverIterationsPreUpdate=1
    simulationSpace=World
    alphaBoolBlend {blendOption=Linear}
    boundBone {boneName=dyn_Curl_07}
    componentAppliedLinearAccClamp=<5000 5000 5000>
    componentLinearAccScale=<0.5 0.5 0.5>
    componentLinearVelScale=<0.5 0.5 0.5>
    gravityOverride {z=-980}
    physicsBodyDefinitions [{
      boundBone {boneName=dyn_Curl_07}
      boxExtents=<5 5 5>
      constraintSetup {angularLimitsMax=<20 20 20>, angularLimitsMin=<-10 -20 -20>, angularTarget=<1 0 0>}
      localJointOffset=<0.2 0 0>
     }]
   }
   n8 = Anim Dynamics - Bone: dyn_Curl_08(ComponentPose=@n7, Alpha=1, ExternalForce="0,0,0") {
    angularDampingOverride=0.999
    angularSpringConstant=150
    bOverrideAngularDamping=true
    bUseGravityOverride=true
    bUsePlanarLimit=false
    gravityScale=0.35
    numSolverIterationsPreUpdate=1
    simulationSpace=World
    alphaBoolBlend {blendOption=Linear}
    boundBone {boneName=dyn_Curl_08}
    componentAppliedLinearAccClamp=<5000 5000 5000>
    componentLinearAccScale=<0.5 0.5 0.5>
    componentLinearVelScale=<0.5 0.5 0.5>
    gravityOverride {z=-980}
    physicsBodyDefinitions [{
      boundBone {boneName=dyn_Curl_08}
      boxExtents=<10 10 10>
      constraintSetup {angularLimitsMax=<20 40 20>, angularLimitsMin=<-10 -40 -20>, angularTarget=<0 0 0>}
      localJointOffset=<0.2 0 0>
     }]
   }
   n9 = Anim Dynamics - Bone: dyn_Curl_09(ComponentPose=@n8, Alpha=1, ExternalForce="0,0,0") {
    angularDampingOverride=0.999
    angularSpringConstant=150
    bOverrideAngularDamping=true
    bUseGravityOverride=true
    bUsePlanarLimit=false
    gravityScale=0.35
    numSolverIterationsPreUpdate=1
    simulationSpace=World
    alphaBoolBlend {blendOption=Linear}
    boundBone {boneName=dyn_Curl_09}
    componentAppliedLinearAccClamp=<5000 5000 5000>
    componentLinearAccScale=<0.5 0.5 0.5>
    componentLinearVelScale=<0.5 0.5 0.5>
    gravityOverride {z=-980}
    physicsBodyDefinitions [{
      boundBone {boneName=dyn_Curl_09}
      boxExtents=<10 10 10>
      constraintSetup {angularLimitsMax=<20 40 20>, angularLimitsMin=<-10 -40 -20>, angularTarget=<1 0 0>}
      localJointOffset=<0.2 0 0>
     }]
   }
   n10 = Anim Dynamics - Bone: dyn_Curl_10(ComponentPose=@n9, Alpha=1, ExternalForce="0,0,0") {
    angularDampingOverride=0.999
    angularSpringConstant=150
    bOverrideAngularDamping=true
    bUseGravityOverride=true
    bUsePlanarLimit=false
    gravityScale=0.35
    numSolverIterationsPreUpdate=1
    simulationSpace=World
    alphaBoolBlend {blendOption=Linear}
    alphaScaleBiasClamp {scale=0.5}
    boundBone {boneName=dyn_Curl_10}
    componentAppliedLinearAccClamp=<5000 5000 5000>
    componentLinearAccScale=<0.5 0.5 0.5>
    componentLinearVelScale=<0.5 0.5 0.5>
    gravityOverride {z=-980}
    physicsBodyDefinitions [{
      boundBone {boneName=dyn_Curl_10}
      boxExtents=<5 5 5>
      constraintSetup {angularLimitsMax=<20 10 10>, angularLimitsMin=<-10 0 0>, angularTarget=<1 0 0>}
      localJointOffset=<0.2 0 0>
     }]
   }
   n11 = Anim Dynamics - Bone: dyn_Curl_11(ComponentPose=@n10, Alpha=1, ExternalForce="0,0,0") {
    angularDampingOverride=0.999
    angularSpringConstant=150
    bOverrideAngularDamping=true
    bUseGravityOverride=true
    bUsePlanarLimit=false
    gravityScale=0.35
    numSolverIterationsPreUpdate=1
    simulationSpace=World
    alphaBoolBlend {blendOption=Linear}
    boundBone {boneName=dyn_Curl_11}
    componentAppliedLinearAccClamp=<5000 5000 5000>
    componentLinearAccScale=<0.5 0.5 0.5>
    componentLinearVelScale=<0.5 0.5 0.5>
    gravityOverride {z=-980}
    physicsBodyDefinitions [{
      boundBone {boneName=dyn_Curl_11}
      boxExtents=<10 10 10>
      constraintSetup {angularLimitsMax=<20 40 20>, angularLimitsMin=<-10 -40 -20>, angularTarget=<1 0 0>}
      localJointOffset=<0.2 0 0>
     }]
   }
   n12 = Anim Dynamics - Bone: dyn_Curl_12(ComponentPose=@n11, Alpha=1, ExternalForce="0,0,0") {
    angularDampingOverride=0.999
    angularSpringConstant=150
    bOverrideAngularDamping=true
    bUseGravityOverride=true
    bUsePlanarLimit=false
    gravityScale=0.35
    numSolverIterationsPreUpdate=1
    simulationSpace=World
    alphaBoolBlend {blendOption=Linear}
    boundBone {boneName=dyn_Curl_12}
    componentAppliedLinearAccClamp=<5000 5000 5000>
    componentLinearAccScale=<0.5 0.5 0.5>
    componentLinearVelScale=<0.5 0.5 0.5>
    gravityOverride {z=-980}
    physicsBodyDefinitions [{
      boundBone {boneName=dyn_Curl_12}
      boxExtents=<30 30 30>
      constraintSetup {angularLimitsMax=<20 30 20>, angularLimitsMin=<-10 -60 -20>, angularTarget=<1 0 0>}
      localJointOffset=<0.2 0 0>
     }]
   }
   n13 = Anim Dynamics - Bone: dyn_Curl_13(ComponentPose=@n12, Alpha=1, ExternalForce="0,0,0") {
    angularDampingOverride=0.999
    angularSpringConstant=150
    bOverrideAngularDamping=true
    bUseGravityOverride=true
    bUsePlanarLimit=false
    gravityScale=0.35
    numSolverIterationsPreUpdate=1
    simulationSpace=World
    alphaBoolBlend {blendOption=Linear}
    boundBone {boneName=dyn_Curl_13}
    componentAppliedLinearAccClamp=<5000 5000 5000>
    componentLinearAccScale=<0.5 0.5 0.5>
    componentLinearVelScale=<0.5 0.5 0.5>
    gravityOverride {z=-980}
    physicsBodyDefinitions [{
      boundBone {boneName=dyn_Curl_13}
      boxExtents=<5 5 5>
      constraintSetup {angularLimitsMax=<20 10 20>, angularLimitsMin=<-10 -40 -20>, angularTarget=<1 0 0>}
      localJointOffset=<0.2 0 0>
     }]
   }
   n14 = Anim Dynamics - Bone: dyn_Curl_14(ComponentPose=@n13, Alpha=1, ExternalForce="0,0,0") {
    angularDampingOverride=0.999
    angularSpringConstant=150
    bOverrideAngularDamping=true
    bUseGravityOverride=true
    bUsePlanarLimit=false
    gravityScale=0.35
    numSolverIterationsPreUpdate=1
    simulationSpace=World
    alphaBoolBlend {blendOption=Linear}
    boundBone {boneName=dyn_Curl_14}
    componentAppliedLinearAccClamp=<5000 5000 5000>
    componentLinearAccScale=<0.5 0.5 0.5>
    componentLinearVelScale=<0.5 0.5 0.5>
    gravityOverride {z=-980}
    physicsBodyDefinitions [{
      boundBone {boneName=dyn_Curl_14}
      boxExtents=<10 10 10>
      constraintSetup {angularLimitsMax=<20 40 20>, angularLimitsMin=<-10 -40 -20>, angularTarget=<1 0 0>}
      localJointOffset=<0.2 0 0>
     }]
   }
   n15 = Anim Dynamics - Bone: dyn_Curl_15(ComponentPose=@n14, Alpha=1, ExternalForce="0,0,0") {
    angularDampingOverride=0.999
    angularSpringConstant=150
    bOverrideAngularDamping=true
    bUseGravityOverride=true
    bUsePlanarLimit=false
    gravityScale=0.35
    numSolverIterationsPreUpdate=1
    simulationSpace=World
    alphaBoolBlend {blendOption=Linear}
    boundBone {boneName=dyn_Curl_15}
    componentAppliedLinearAccClamp=<5000 5000 5000>
    componentLinearAccScale=<0.5 0.5 0.5>
    componentLinearVelScale=<0.5 0.5 0.5>
    gravityOverride {z=-980}
    physicsBodyDefinitions [{
      boundBone {boneName=dyn_Curl_15}
      boxExtents=<2 2 2>
      constraintSetup {angularLimitsMax=<5 30 20>, angularLimitsMin=<-5 -50 -20>, angularTarget=<1 0 0>}
      localJointOffset=<0.2 0 0>
     }]
   }
   n16 = Component To Local(ComponentPose=@n15)
   n17 = Anim Dynamics - Bone: dyn_Curl_01(ComponentPose=Local To Component(LocalPose=Use cached pose 'Input Pose Cache'()), Alpha=1, ExternalForce="0,0,0") {
    angularDampingOverride=0.9999
    angularSpringConstant=150
    bOverrideAngularDamping=true
    bUseGravityOverride=true
    bUsePlanarLimit=false
    gravityScale=0.35
    numSolverIterationsPreUpdate=1
    simulationSpace=World
    alphaBoolBlend {blendOption=Linear}
    boundBone {boneName=dyn_Curl_01}
    componentAppliedLinearAccClamp=<5000 5000 5000>
    componentLinearAccScale=<0.5 0.5 0.5>
    componentLinearVelScale=<0.5 0.5 0.5>
    gravityOverride {z=-980}
    physicsBodyDefinitions [{
      boundBone {boneName=dyn_Curl_01}
      boxExtents=<5 5 5>
      constraintSetup {angularLimitsMax=<20 30 20>, angularLimitsMin=<-10 -50 -20>, angularTarget=<1 0 0>}
      localJointOffset=<0 0.4 0>
     }]
   }
   n18 = Anim Dynamics - Bone: dyn_Curl_02(ComponentPose=@n17, Alpha=1, ExternalForce="0,0,0") {
    angularDampingOverride=0.9999
    angularSpringConstant=150
    bOverrideAngularDamping=true
    bUseGravityOverride=true
    bUsePlanarLimit=false
    gravityScale=0.35
    numSolverIterationsPreUpdate=1
    simulationSpace=World
    alphaBoolBlend {blendOption=Linear}
    boundBone {boneName=dyn_Curl_02}
    componentAppliedLinearAccClamp=<5000 5000 5000>
    componentLinearAccScale=<0.5 0.5 0.5>
    componentLinearVelScale=<0.5 0.5 0.5>
    gravityOverride {z=-980}
    physicsBodyDefinitions [{
      boundBone {boneName=dyn_Curl_02}
      boxExtents=<10 10 10>
      constraintSetup {angularLimitsMax=<20 50 20>, angularLimitsMin=<-10 -50 -20>, angularTarget=<1 0 0>}
      localJointOffset=<0 0.4 0>
     }]
   }
   n19 = Anim Dynamics - Bone: dyn_Curl_03(ComponentPose=@n18, Alpha=1, ExternalForce="0,0,0") {
    angularDampingOverride=0.9999
    angularSpringConstant=150
    bOverrideAngularDamping=true
    bUseGravityOverride=true
    bUsePlanarLimit=false
    gravityScale=0.35
    numSolverIterationsPreUpdate=1
    simulationSpace=World
    alphaBoolBlend {blendOption=Linear}
    boundBone {boneName=dyn_Curl_03}
    componentAppliedLinearAccClamp=<5000 5000 5000>
    componentLinearAccScale=<0.5 0.5 0.5>
    componentLinearVelScale=<0.5 0.5 0.5>
    gravityOverride {z=-980}
    physicsBodyDefinitions [{
      boundBone {boneName=dyn_Curl_03}
      boxExtents=<2 2 2>
      constraintSetup {angularLimitsMax=<20 0 20>, angularLimitsMin=<-10 -50 -20>, angularTarget=<1 0 0>}
      localJointOffset=<0 0.4 0>
     }]
   }
   n20 = Anim Dynamics - Bone: dyn_Curl_04(ComponentPose=@n19, Alpha=1, ExternalForce="0,0,0") {
    angularDampingOverride=0.9999
    angularSpringConstant=150
    bOverrideAngularDamping=true
    bUseGravityOverride=true
    bUsePlanarLimit=false
    gravityScale=0.35
    numSolverIterationsPreUpdate=1
    simulationSpace=World
    alphaBoolBlend {blendOption=Linear}
    boundBone {boneName=dyn_Curl_04}
    componentAppliedLinearAccClamp=<5000 5000 5000>
    componentLinearAccScale=<0.5 0.5 0.5>
    componentLinearVelScale=<0.5 0.5 0.5>
    gravityOverride {z=-980}
    physicsBodyDefinitions [{
      boundBone {boneName=dyn_Curl_04}
      boxExtents=<5 5 5>
      constraintSetup {angularLimitsMax=<20 30 20>, angularLimitsMin=<-10 -40 -20>, angularTarget=<1 0 0>}
      localJointOffset=<0 0.4 0>
     }]
   }
   n21 = Anim Dynamics - Bone: dyn_Curl_05(ComponentPose=@n20, Alpha=1, ExternalForce="0,0,0") {
    angularDampingOverride=0.9999
    angularSpringConstant=150
    bOverrideAngularDamping=true
    bUseGravityOverride=true
    bUsePlanarLimit=false
    gravityScale=0.35
    numSolverIterationsPreUpdate=1
    simulationSpace=World
    alphaBoolBlend {blendOption=Linear}
    boundBone {boneName=dyn_Curl_05}
    componentAppliedLinearAccClamp=<5000 5000 5000>
    componentLinearAccScale=<0.5 0.5 0.5>
    componentLinearVelScale=<0.5 0.5 0.5>
    gravityOverride {z=-980}
    physicsBodyDefinitions [{
      boundBone {boneName=dyn_Curl_05}
      boxExtents=<10 10 10>
      constraintSetup {angularLimitsMax=<20 50 20>, angularLimitsMin=<-10 -50 -20>, angularTarget=<1 0 0>}
      localJointOffset=<0 0.4 0>
     }]
   }
   n22 = Anim Dynamics - Bone: dyn_Curl_06(ComponentPose=@n21, Alpha=1, ExternalForce="0,0,0") {
    angularDampingOverride=0.9999
    angularSpringConstant=150
    bOverrideAngularDamping=true
    bUseGravityOverride=true
    bUsePlanarLimit=false
    gravityScale=0.35
    numSolverIterationsPreUpdate=1
    simulationSpace=World
    alphaBoolBlend {blendOption=Linear}
    boundBone {boneName=dyn_Curl_06}
    componentAppliedLinearAccClamp=<5000 5000 5000>
    componentLinearAccScale=<0.5 0.5 0.5>
    componentLinearVelScale=<0.5 0.5 0.5>
    gravityOverride {z=-980}
    physicsBodyDefinitions [{
      boundBone {boneName=dyn_Curl_06}
      boxExtents=<2 2 2>
      constraintSetup {angularLimitsMax=<20 30 30>, angularLimitsMin=<-10 -30 -30>, angularTarget=<1 0 0>}
      localJointOffset=<0 0.4 0>
     }]
   }
   n23 = Anim Dynamics - Bone: dyn_Curl_07(ComponentPose=@n22, Alpha=1, ExternalForce="0,0,0") {
    angularDampingOverride=0.9999
    angularSpringConstant=150
    bOverrideAngularDamping=true
    bUseGravityOverride=true
    bUsePlanarLimit=false
    gravityScale=0.35
    numSolverIterationsPreUpdate=1
    simulationSpace=World
    alphaBoolBlend {blendOption=Linear}
    boundBone {boneName=dyn_Curl_07}
    componentAppliedLinearAccClamp=<5000 5000 5000>
    componentLinearAccScale=<0.5 0.5 0.5>
    componentLinearVelScale=<0.5 0.5 0.5>
    gravityOverride {z=-980}
    physicsBodyDefinitions [{
      boundBone {boneName=dyn_Curl_07}
      boxExtents=<5 5 5>
      constraintSetup {angularLimitsMax=<20 20 20>, angularLimitsMin=<-10 -20 -20>, angularTarget=<1 0 0>}
      localJointOffset=<0 0.4 0>
     }]
   }
   n24 = Anim Dynamics - Bone: dyn_Curl_08(ComponentPose=@n23, Alpha=1, ExternalForce="0,0,0") {
    angularDampingOverride=0.9999
    angularSpringConstant=150
    bOverrideAngularDamping=true
    bUseGravityOverride=true
    bUsePlanarLimit=false
    gravityScale=0.35
    numSolverIterationsPreUpdate=1
    simulationSpace=World
    alphaBoolBlend {blendOption=Linear}
    boundBone {boneName=dyn_Curl_08}
    componentAppliedLinearAccClamp=<5000 5000 5000>
    componentLinearAccScale=<0.5 0.5 0.5>
    componentLinearVelScale=<0.5 0.5 0.5>
    gravityOverride {z=-980}
    physicsBodyDefinitions [{
      boundBone {boneName=dyn_Curl_08}
      boxExtents=<10 10 10>
      constraintSetup {angularLimitsMax=<20 40 20>, angularLimitsMin=<-10 -40 -20>, angularTarget=<0 0 0>}
      localJointOffset=<0 0.4 0>
     }]
   }
   n25 = Anim Dynamics - Bone: dyn_Curl_09(ComponentPose=@n24, Alpha=1, ExternalForce="0,0,0") {
    angularDampingOverride=0.9999
    angularSpringConstant=150
    bOverrideAngularDamping=true
    bUseGravityOverride=true
    bUsePlanarLimit=false
    gravityScale=0.35
    numSolverIterationsPreUpdate=1
    simulationSpace=World
    alphaBoolBlend {blendOption=Linear}
    boundBone {boneName=dyn_Curl_09}
    componentAppliedLinearAccClamp=<5000 5000 5000>
    componentLinearAccScale=<0.5 0.5 0.5>
    componentLinearVelScale=<0.5 0.5 0.5>
    gravityOverride {z=-980}
    physicsBodyDefinitions [{
      boundBone {boneName=dyn_Curl_09}
      boxExtents=<10 10 10>
      constraintSetup {angularLimitsMax=<20 40 20>, angularLimitsMin=<-10 -40 -20>, angularTarget=<1 0 0>}
      localJointOffset=<0 0.4 0>
     }]
   }
   n26 = Anim Dynamics - Bone: dyn_Curl_10(ComponentPose=@n25, Alpha=1, ExternalForce="0,0,0") {
    angularDampingOverride=0.9999
    angularSpringConstant=150
    bOverrideAngularDamping=true
    bUseGravityOverride=true
    bUsePlanarLimit=false
    gravityScale=0.35
    numSolverIterationsPreUpdate=1
    simulationSpace=World
    alphaBoolBlend {blendOption=Linear}
    alphaScaleBiasClamp {scale=0.5}
    boundBone {boneName=dyn_Curl_10}
    componentAppliedLinearAccClamp=<5000 5000 5000>
    componentLinearAccScale=<0.5 0.5 0.5>
    componentLinearVelScale=<0.5 0.5 0.5>
    gravityOverride {z=-980}
    physicsBodyDefinitions [{
      boundBone {boneName=dyn_Curl_10}
      boxExtents=<5 5 5>
      constraintSetup {angularLimitsMax=<20 10 10>, angularLimitsMin=<-10 0 0>, angularTarget=<1 0 0>}
      localJointOffset=<0 0.4 0>
     }]
   }
   n27 = Anim Dynamics - Bone: dyn_Curl_11(ComponentPose=@n26, Alpha=1, ExternalForce="0,0,0") {
    angularDampingOverride=0.9999
    angularSpringConstant=150
    bOverrideAngularDamping=true
    bUseGravityOverride=true
    bUsePlanarLimit=false
    gravityScale=0.35
    numSolverIterationsPreUpdate=1
    simulationSpace=World
    alphaBoolBlend {blendOption=Linear}
    boundBone {boneName=dyn_Curl_11}
    componentAppliedLinearAccClamp=<5000 5000 5000>
    componentLinearAccScale=<0.5 0.5 0.5>
    componentLinearVelScale=<0.5 0.5 0.5>
    gravityOverride {z=-980}
    physicsBodyDefinitions [{
      boundBone {boneName=dyn_Curl_11}
      boxExtents=<10 10 10>
      constraintSetup {angularLimitsMax=<20 40 20>, angularLimitsMin=<-10 -40 -20>, angularTarget=<1 0 0>}
      localJointOffset=<0 0.4 0>
     }]
   }
   n28 = Anim Dynamics - Bone: dyn_Curl_12(ComponentPose=@n27, Alpha=1, ExternalForce="0,0,0") {
    angularDampingOverride=0.9999
    angularSpringConstant=150
    bOverrideAngularDamping=true
    bUseGravityOverride=true
    bUsePlanarLimit=false
    gravityScale=0.35
    numSolverIterationsPreUpdate=1
    simulationSpace=World
    alphaBoolBlend {blendOption=Linear}
    boundBone {boneName=dyn_Curl_12}
    componentAppliedLinearAccClamp=<5000 5000 5000>
    componentLinearAccScale=<0.5 0.5 0.5>
    componentLinearVelScale=<0.5 0.5 0.5>
    gravityOverride {z=-980}
    physicsBodyDefinitions [{
      boundBone {boneName=dyn_Curl_12}
      boxExtents=<30 30 30>
      constraintSetup {angularLimitsMax=<20 30 20>, angularLimitsMin=<-10 -60 -20>, angularTarget=<1 0 0>}
      localJointOffset=<0 0.4 0>
     }]
   }
   n29 = Anim Dynamics - Bone: dyn_Curl_13(ComponentPose=@n28, Alpha=1, ExternalForce="0,0,0") {
    angularDampingOverride=0.9999
    angularSpringConstant=150
    bOverrideAngularDamping=true
    bUseGravityOverride=true
    bUsePlanarLimit=false
    gravityScale=0.35
    numSolverIterationsPreUpdate=1
    simulationSpace=World
    alphaBoolBlend {blendOption=Linear}
    boundBone {boneName=dyn_Curl_13}
    componentAppliedLinearAccClamp=<5000 5000 5000>
    componentLinearAccScale=<0.5 0.5 0.5>
    componentLinearVelScale=<0.5 0.5 0.5>
    gravityOverride {z=-980}
    physicsBodyDefinitions [{
      boundBone {boneName=dyn_Curl_13}
      boxExtents=<5 5 5>
      constraintSetup {angularLimitsMax=<20 10 20>, angularLimitsMin=<-10 -40 -20>, angularTarget=<1 0 0>}
      localJointOffset=<0 0.4 0>
     }]
   }
   n30 = Anim Dynamics - Bone: dyn_Curl_14(ComponentPose=@n29, Alpha=1, ExternalForce="0,0,0") {
    angularDampingOverride=0.9999
    angularSpringConstant=150
    bOverrideAngularDamping=true
    bUseGravityOverride=true
    bUsePlanarLimit=false
    gravityScale=0.35
    numSolverIterationsPreUpdate=1
    simulationSpace=World
    alphaBoolBlend {blendOption=Linear}
    boundBone {boneName=dyn_Curl_14}
    componentAppliedLinearAccClamp=<5000 5000 5000>
    componentLinearAccScale=<0.5 0.5 0.5>
    componentLinearVelScale=<0.5 0.5 0.5>
    gravityOverride {z=-980}
    physicsBodyDefinitions [{
      boundBone {boneName=dyn_Curl_14}
      boxExtents=<10 10 10>
      constraintSetup {angularLimitsMax=<20 40 20>, angularLimitsMin=<-10 -40 -20>, angularTarget=<1 0 0>}
      localJointOffset=<0 0.4 0>
     }]
   }
   n31 = Anim Dynamics - Bone: dyn_Curl_15(ComponentPose=@n30, Alpha=1, ExternalForce="0,0,0") {
    angularDampingOverride=0.9999
    angularSpringConstant=150
    bOverrideAngularDamping=true
    bUseGravityOverride=true
    bUsePlanarLimit=false
    gravityScale=0.35
    numSolverIterationsPreUpdate=1
    simulationSpace=World
    alphaBoolBlend {blendOption=Linear}
    boundBone {boneName=dyn_Curl_15}
    componentAppliedLinearAccClamp=<5000 5000 5000>
    componentLinearAccScale=<0.5 0.5 0.5>
    componentLinearVelScale=<0.5 0.5 0.5>
    gravityOverride {z=-980}
    physicsBodyDefinitions [{
      boundBone {boneName=dyn_Curl_15}
      boxExtents=<2 2 2>
      constraintSetup {angularLimitsMax=<5 30 20>, angularLimitsMin=<-5 -50 -20>, angularTarget=<1 0 0>}
      localJointOffset=<0 0.4 0>
     }]
   }
   n32 = Component To Local(ComponentPose=@n31)
   n33 = Get bIsOnPawn() -> (bIsOnPawn:bool)
   n34 = Anim Dynamics - Bone: dyn_Curl_01(ComponentPose=Local To Component(LocalPose=Use cached pose 'Input Pose Cache'()), Alpha=1, ExternalForce="0,0,0") {
    angularDampingOverride=0.999
    angularSpringConstant=5000
    bAngularSpring=true
    bOverrideAngularDamping=true
    bUseGravityOverride=true
    bUsePlanarLimit=false
    gravityScale=0.35
    numSolverIterationsPreUpdate=1
    simulationSpace=RootRelative
    alphaBoolBlend {blendOption=Linear}
    boundBone {boneName=dyn_Curl_01}
    componentAppliedLinearAccClamp=<1300 1300 1300>
    componentLinearAccScale=<0.5 0.5 0.5>
    componentLinearVelScale=<0.5 0.5 0.5>
    gravityOverride {z=-980}
    physicsBodyDefinitions [{
      boundBone {boneName=dyn_Curl_01}
      boxExtents=<30 30 30>
      constraintSetup {angularLimitsMax=<20 30 25>, angularLimitsMin=<-20 -50 -25>, angularTarget=<1 0 0>}
      localJointOffset=<-2 0.2 0.1>
     }]
   }
   n35 = Anim Dynamics - Bone: dyn_Curl_02(ComponentPose=@n34, Alpha=1, ExternalForce="0,0,0") {
    angularDampingOverride=0.999
    angularSpringConstant=5000
    bAngularSpring=true
    bOverrideAngularDamping=true
    bUseGravityOverride=true
    bUsePlanarLimit=false
    gravityScale=0.35
    numSolverIterationsPreUpdate=1
    simulationSpace=RootRelative
    alphaBoolBlend {blendOption=Linear}
    boundBone {boneName=dyn_Curl_02}
    componentAppliedLinearAccClamp=<1300 1300 1300>
    componentLinearAccScale=<0.5 0.5 0.5>
    componentLinearVelScale=<0.5 0.5 0.5>
    gravityOverride {z=-980}
    physicsBodyDefinitions [{
      boundBone {boneName=dyn_Curl_02}
      boxExtents=<30 30 30>
      constraintSetup {angularLimitsMax=<20 40 30>, angularLimitsMin=<-20 -40 -15>, angularTarget=<1 0 0>}
      localJointOffset=<-2 0.2 0.1>
     }]
   }
   n36 = Anim Dynamics - Bone: dyn_Curl_03(ComponentPose=@n35, Alpha=1, ExternalForce="0,0,0") {
    angularDampingOverride=0.999
    angularSpringConstant=5000
    bAngularSpring=true
    bOverrideAngularDamping=true
    bUseGravityOverride=true
    bUsePlanarLimit=false
    gravityScale=0.35
    numSolverIterationsPreUpdate=1
    simulationSpace=RootRelative
    alphaBoolBlend {blendOption=Linear}
    boundBone {boneName=dyn_Curl_03}
    componentAppliedLinearAccClamp=<1300 1300 1300>
    componentLinearAccScale=<0.5 0.5 0.5>
    componentLinearVelScale=<0.5 0.5 0.5>
    gravityOverride {z=-980}
    physicsBodyDefinitions [{
      boundBone {boneName=dyn_Curl_03}
      boxExtents=<30 30 30>
      constraintSetup {angularLimitsMax=<20 0 20>, angularLimitsMin=<-20 -50 -20>, angularTarget=<1 0 0>}
      localJointOffset=<-2 0.2 0.1>
     }]
   }
   n37 = Anim Dynamics - Bone: dyn_Curl_04(ComponentPose=@n36, Alpha=1, ExternalForce="0,0,0") {
    angularDampingOverride=0.999
    angularSpringConstant=5000
    bAngularSpring=true
    bOverrideAngularDamping=true
    bUseGravityOverride=true
    bUsePlanarLimit=false
    gravityScale=0.35
    numSolverIterationsPreUpdate=1
    simulationSpace=RootRelative
    alphaBoolBlend {blendOption=Linear}
    boundBone {boneName=dyn_Curl_04}
    componentAppliedLinearAccClamp=<1300 1300 1300>
    componentLinearAccScale=<0.5 0.5 0.5>
    componentLinearVelScale=<0.5 0.5 0.5>
    gravityOverride {z=-980}
    physicsBodyDefinitions [{
      boundBone {boneName=dyn_Curl_04}
      boxExtents=<30 30 30>
      constraintSetup {angularLimitsMax=<20 30 20>, angularLimitsMin=<-20 -40 -20>, angularTarget=<1 0 0>}
      localJointOffset=<-2 0.2 0.1>
     }]
   }
   n38 = Anim Dynamics - Bone: dyn_Curl_05(ComponentPose=@n37, Alpha=1, ExternalForce="0,0,0") {
    angularDampingOverride=0.999
    angularSpringConstant=5000
    bAngularSpring=true
    bOverrideAngularDamping=true
    bUseGravityOverride=true
    bUsePlanarLimit=false
    gravityScale=0.35
    numSolverIterationsPreUpdate=1
    simulationSpace=RootRelative
    alphaBoolBlend {blendOption=Linear}
    boundBone {boneName=dyn_Curl_05}
    componentAppliedLinearAccClamp=<1300 1300 1300>
    componentLinearAccScale=<0.5 0.5 0.5>
    componentLinearVelScale=<0.5 0.5 0.5>
    gravityOverride {z=-980}
    physicsBodyDefinitions [{
      boundBone {boneName=dyn_Curl_05}
      boxExtents=<30 30 30>
      constraintSetup {angularLimitsMax=<20 40 20>, angularLimitsMin=<-20 -50 -20>, angularTarget=<1 0 0>}
      localJointOffset=<-2 0.2 0.1>
     }]
   }
   n39 = Anim Dynamics - Bone: dyn_Curl_06(ComponentPose=@n38, Alpha=1, ExternalForce="0,0,0") {
    angularDampingOverride=0.999
    angularSpringConstant=5000
    bAngularSpring=true
    bOverrideAngularDamping=true
    bUseGravityOverride=true
    bUsePlanarLimit=false
    gravityScale=0.35
    numSolverIterationsPreUpdate=1
    simulationSpace=RootRelative
    alphaBoolBlend {blendOption=Linear}
    boundBone {boneName=dyn_Curl_06}
    componentAppliedLinearAccClamp=<1300 1300 1300>
    componentLinearAccScale=<0.5 0.5 0.5>
    componentLinearVelScale=<0.5 0.5 0.5>
    gravityOverride {z=-980}
    physicsBodyDefinitions [{
      boundBone {boneName=dyn_Curl_06}
      boxExtents=<30 30 30>
      constraintSetup {angularLimitsMax=<20 30 20>, angularLimitsMin=<-20 -30 -20>, angularTarget=<1 0 0>}
      localJointOffset=<-2 0.2 0.1>
     }]
   }
   n40 = Anim Dynamics - Bone: dyn_Curl_07(ComponentPose=@n39, Alpha=1, ExternalForce="0,0,0") {
    angularDampingOverride=0.999
    angularSpringConstant=5000
    bAngularSpring=true
    bOverrideAngularDamping=true
    bUseGravityOverride=true
    bUsePlanarLimit=false
    gravityScale=0.35
    numSolverIterationsPreUpdate=1
    simulationSpace=RootRelative
    alphaBoolBlend {blendOption=Linear}
    boundBone {boneName=dyn_Curl_07}
    componentAppliedLinearAccClamp=<1300 1300 1300>
    componentLinearAccScale=<0.5 0.5 0.5>
    componentLinearVelScale=<0.5 0.5 0.5>
    gravityOverride {z=-980}
    physicsBodyDefinitions [{
      boundBone {boneName=dyn_Curl_07}
      boxExtents=<30 30 30>
      constraintSetup {angularLimitsMax=<20 20 10>, angularLimitsMin=<-20 -20 -10>, angularTarget=<1 0 0>}
      localJointOffset=<-2 0.2 0.1>
     }]
   }
   n41 = Anim Dynamics - Bone: dyn_Curl_08(ComponentPose=@n40, Alpha=1, ExternalForce="0,0,0") {
    angularDampingOverride=0.999
    angularSpringConstant=5000
    bAngularSpring=true
    bOverrideAngularDamping=true
    bUseGravityOverride=true
    bUsePlanarLimit=false
    gravityScale=0.35
    numSolverIterationsPreUpdate=1
    simulationSpace=RootRelative
    alphaBoolBlend {blendOption=Linear}
    boundBone {boneName=dyn_Curl_08}
    componentAppliedLinearAccClamp=<1300 1300 1300>
    componentLinearAccScale=<0.5 0.5 0.5>
    componentLinearVelScale=<0.5 0.5 0.5>
    gravityOverride {z=-980}
    physicsBodyDefinitions [{
      boundBone {boneName=dyn_Curl_08}
      boxExtents=<30 30 30>
      constraintSetup {angularLimitsMax=<20 20 20>, angularLimitsMin=<-20 -60 -20>, angularTarget=<1 0 0>}
      localJointOffset=<-2 0.2 0.1>
     }]
   }
   n42 = Anim Dynamics - Bone: dyn_Curl_09(ComponentPose=@n41, Alpha=1, ExternalForce="0,0,0") {
    angularDampingOverride=0.999
    angularSpringConstant=5000
    bAngularSpring=true
    bOverrideAngularDamping=true
    bUseGravityOverride=true
    bUsePlanarLimit=false
    gravityScale=0.35
    numSolverIterationsPreUpdate=1
    simulationSpace=RootRelative
    alphaBoolBlend {blendOption=Linear}
    boundBone {boneName=dyn_Curl_09}
    componentAppliedLinearAccClamp=<1300 1300 1300>
    componentLinearAccScale=<0.5 0.5 0.5>
    componentLinearVelScale=<0.5 0.5 0.5>
    gravityOverride {z=-980}
    physicsBodyDefinitions [{
      boundBone {boneName=dyn_Curl_09}
      boxExtents=<30 30 30>
      constraintSetup {angularLimitsMax=<20 40 20>, angularLimitsMin=<-20 -40 -10>, angularTarget=<1 0 0>}
      localJointOffset=<-2 0.2 0.1>
     }]
   }
   n43 = Anim Dynamics - Bone: dyn_Curl_10(ComponentPose=@n42, Alpha=1, ExternalForce="0,0,0") {
    angularDampingOverride=0.999
    angularSpringConstant=5000
    bAngularSpring=true
    bOverrideAngularDamping=true
    bUseGravityOverride=true
    bUsePlanarLimit=false
    gravityScale=0.35
    numSolverIterationsPreUpdate=1
    simulationSpace=RootRelative
    alphaBoolBlend {blendOption=Linear}
    boundBone {boneName=dyn_Curl_10}
    componentAppliedLinearAccClamp=<1300 1300 1300>
    componentLinearAccScale=<0.5 0.5 0.5>
    componentLinearVelScale=<0.5 0.5 0.5>
    gravityOverride {z=-980}
    physicsBodyDefinitions [{
      boundBone {boneName=dyn_Curl_10}
      boxExtents=<30 30 30>
      constraintSetup {angularLimitsMax=<20 -5 20>, angularLimitsMin=<-20 -60 -20>, angularTarget=<1 0 0>}
      localJointOffset=<-2 0.2 0.1>
     }]
   }
   n44 = Anim Dynamics - Bone: dyn_Curl_11(ComponentPose=@n43, Alpha=1, ExternalForce="0,0,0") {
    angularDampingOverride=0.999
    angularSpringConstant=5000
    bAngularSpring=true
    bOverrideAngularDamping=true
    bUseGravityOverride=true
    bUsePlanarLimit=false
    gravityScale=0.35
    numSolverIterationsPreUpdate=1
    simulationSpace=RootRelative
    alphaBoolBlend {blendOption=Linear}
    boundBone {boneName=dyn_Curl_11}
    componentAppliedLinearAccClamp=<1300 1300 1300>
    componentLinearAccScale=<0.5 0.5 0.5>
    componentLinearVelScale=<0.5 0.5 0.5>
    gravityOverride {z=-980}
    physicsBodyDefinitions [{
      boundBone {boneName=dyn_Curl_11}
      boxExtents=<30 30 30>
      constraintSetup {angularLimitsMax=<20 40 20>, angularLimitsMin=<-20 -40 -20>, angularTarget=<1 0 0>}
      localJointOffset=<-2 0.2 0.1>
     }]
   }
   n45 = Anim Dynamics - Bone: dyn_Curl_12(ComponentPose=@n44, Alpha=1, ExternalForce="0,0,0") {
    angularDampingOverride=0.999
    angularSpringConstant=5000
    bAngularSpring=true
    bOverrideAngularDamping=true
    bUseGravityOverride=true
    bUsePlanarLimit=false
    gravityScale=0.35
    numSolverIterationsPreUpdate=1
    simulationSpace=RootRelative
    alphaBoolBlend {blendOption=Linear}
    boundBone {boneName=dyn_Curl_12}
    componentAppliedLinearAccClamp=<1300 1300 1300>
    componentLinearAccScale=<0.5 0.5 0.5>
    componentLinearVelScale=<0.5 0.5 0.5>
    gravityOverride {z=-980}
    physicsBodyDefinitions [{
      boundBone {boneName=dyn_Curl_12}
      boxExtents=<30 30 30>
      constraintSetup {angularLimitsMax=<20 30 20>, angularLimitsMin=<-20 -60 -20>, angularTarget=<1 0 0>}
      localJointOffset=<-2 0.2 0.1>
     }]
   }
   n46 = Anim Dynamics - Bone: dyn_Curl_13(ComponentPose=@n45, Alpha=1, ExternalForce="0,0,0") {
    angularDampingOverride=0.999
    angularSpringConstant=5000
    bAngularSpring=true
    bOverrideAngularDamping=true
    bUseGravityOverride=true
    bUsePlanarLimit=false
    gravityScale=0.35
    numSolverIterationsPreUpdate=1
    simulationSpace=RootRelative
    alphaBoolBlend {blendOption=Linear}
    boundBone {boneName=dyn_Curl_13}
    componentAppliedLinearAccClamp=<1300 1300 1300>
    componentLinearAccScale=<0.5 0.5 0.5>
    componentLinearVelScale=<0.5 0.5 0.5>
    gravityOverride {z=-980}
    physicsBodyDefinitions [{
      boundBone {boneName=dyn_Curl_13}
      boxExtents=<30 30 30>
      constraintSetup {angularLimitsMax=<20 10 20>, angularLimitsMin=<-20 -40 -20>, angularTarget=<1 0 0>}
      localJointOffset=<-2 0.2 0.1>
     }]
   }
   n47 = Anim Dynamics - Bone: dyn_Curl_14(ComponentPose=@n46, Alpha=1, ExternalForce="0,0,0") {
    angularDampingOverride=0.999
    angularSpringConstant=5000
    bAngularSpring=true
    bOverrideAngularDamping=true
    bUseGravityOverride=true
    bUsePlanarLimit=false
    gravityScale=0.35
    numSolverIterationsPreUpdate=1
    simulationSpace=RootRelative
    alphaBoolBlend {blendOption=Linear}
    boundBone {boneName=dyn_Curl_14}
    componentAppliedLinearAccClamp=<1300 1300 1300>
    componentLinearAccScale=<0.5 0.5 0.5>
    componentLinearVelScale=<0.5 0.5 0.5>
    gravityOverride {z=-980}
    physicsBodyDefinitions [{
      boundBone {boneName=dyn_Curl_14}
      boxExtents=<30 30 30>
      constraintSetup {angularLimitsMax=<20 40 40>, angularLimitsMin=<-20 -40 -40>, angularTarget=<1 0 0>}
      localJointOffset=<-2 0.2 0.1>
     }]
   }
   n48 = Anim Dynamics - Bone: dyn_Curl_15(ComponentPose=@n47, Alpha=1, ExternalForce="0,0,0") {
    angularDampingOverride=0.999
    angularSpringConstant=5000
    bAngularSpring=true
    bOverrideAngularDamping=true
    bUseGravityOverride=true
    bUsePlanarLimit=false
    gravityScale=0.35
    numSolverIterationsPreUpdate=1
    simulationSpace=RootRelative
    alphaBoolBlend {blendOption=Linear}
    boundBone {boneName=dyn_Curl_15}
    componentAppliedLinearAccClamp=<1300 1300 1300>
    componentLinearAccScale=<0.5 0.5 0.5>
    componentLinearVelScale=<0.5 0.5 0.5>
    gravityOverride {z=-980}
    physicsBodyDefinitions [{
      boundBone {boneName=dyn_Curl_15}
      boxExtents=<30 30 30>
      constraintSetup {angularLimitsMax=<20 40 20>, angularLimitsMin=<-20 -30 -20>, angularTarget=<1 0 0>}
      localJointOffset=<-2 0.2 0.1>
     }]
   }
   Output Pose(Result=Blend Poses by bool(
     BlendPose_0=Blend Poses by bool(BlendPose_0=@n32, BlendPose_1=@n16, bActiveValue=@n33) {blendType=Linear}
     BlendPose_1=Component To Local(ComponentPose=@n48)
     bActiveValue=Get bIsInFrontEnd()
    ) {blendType=Linear}) {layerGroup=None}
   AnimGraphNode_SubInput_0 = Input Pose()
   Save cached pose 'Input Pose Cache'(Pose=@AnimGraphNode_SubInput_0) {cacheName=Input Pose Cache}
  ]}
 EventGraph {nodes [
   TryGetPawnOwner()
   Event BlueprintUpdateAnimation()
  ]}
]
variables []
